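# Lit Niagara mesh-particle material for the exploding ice sculpture in the Sculptor 2024 emote, using ice diffuse, SRM and normal textures.
# Per-piece pivots decoded from UV1 and UV2 also seed noise lookups, offset by particle random, for rotation axes and scatter.
# Dynamic parameters drive vertex offsets: outward push, scale, vertical gravity, rotation and random spread.
# The normal map is rotated with each piece to keep shading consistent.
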Material /Game/Animation/Game/MainPlayer/Emotes/Sculptor/FX/M_Emote_Sculptor_2024_Exploding
graph {
 properties {bUsedWithNiagaraMeshParticles=true, pixelDepthOffsetMode=PDOM_Legacy, usedShadingModels=Default Lit}
 nodes [
  n1 = TextureCoordinate() {coordinateIndex=1}
  n2 = ComponentMask(None=@n1) {g=true}
  n3 = ComponentMask(None=TextureCoordinate() {coordinateIndex=2}) {r=true}
  n4 = AppendVector(A=Add(A=ComponentMask(None=@n1) {r=true}, B=@n2), B=Add(A=@n2, B=@n3))
  n5 = TextureSample(UVs=Add(A=Multiply(A=@n4) {constB=0.235}, B=ParticleRandom())) {samplerType=SAMPLERTYPE_LinearColor, texture=&HighContrastNoise}
  n6 = Normalize(VectorInput=Subtract(A=Multiply(A=@n5.RGB) {constB=2}))
  n7 = Transform(None=AppendVector(A=@n1, B=@n3)) {transformSourceType=TRANSFORMSOURCE_Local}
  n8 = ParticlePositionWS()
  n9 = Add(A=@n7, B=@n8)
  n10 = WorldPosition()
  n11 = DynamicParameter() {defaultValue=<rgba 0 1 0 1>, paramNames [PositionOverLifetime, ScaleOverLIfetime, GravityOverLifetime, RotationOverLifetime]}
  n12 = RotateAboutAxis(NormalizedRotationAxis=@n6, RotationAngle=@n11.RotationOverLifetime, PivotPoint=@n9, Position=@n10.XYZ)
  n13 = Add(A=@n12, B=@n10.XYZ)
  n14 = Add(A=@n12, B=Multiply(A=Subtract(A=@n9, B=@n13), B=OneMinus(None=@n11.ScaleOverLIfetime)))
  n15 = Multiply(A=Normalize(VectorInput=@n7), B=@n11.PositionOverLifetime)
  n16 = Constant2Vector()
  n17 = DynamicParameter() {parameterIndex=2, defaultValue=<rgba 0 0 0 1>, paramNames [PositionRandomX, PositionRandomY, PositionRandomZ, Param4]}
  Add(A=Add(A=Add(A=@n15, B=@n14), B=AppendVector(A=@n16, B=@n11.GravityOverLifetime)), B=Multiply(A=@n5.RGB, B=@n17.RGB))
  Constant3Vector() {constant=<rgba 0 0 1 0>}
  Multiply(
   A=TextureSampleParameter2D() {parameterName=Diffuse Texture, texture=/Game/Animation/Props/Gadgets_Sculptor_Ice/Textures/T_Sculptor_Ice_Emote_D}.RGB
   B=ParticleColor().RGB
  )
  n18 = TextureSampleParameter2D() {parameterName=SRM Texture, samplerType=SAMPLERTYPE_Masks, texture=/Game/Animation/Props/Gadgets_Sculptor_Ice/Textures/T_Sculptor_Ice_Emote_S}
  Multiply(A=@n18.B, B=Constant() {r=1.5})
  n19 = DynamicParameter() {parameterIndex=1, defaultValue=<rgba 0 0 0 1>, paramNames [PointPositionX, PointPositionY, PointPositionZ, Param4]}
  Subtract(A=@n9, B=Add(A=@n8, B=@n19.RGB))
  TransformPosition(None=@n10.XYZ) {transformSourceType=TRANSFORMPOSSOURCE_World, transformType=TRANSFORMPOSSOURCE_Instance}
  n20 = TextureSampleParameter2D() {parameterName=Normal Map, samplerType=SAMPLERTYPE_Normal, texture=/Game/Animation/Props/Gadgets_Sculptor_Ice/Textures/T_Sculptor_Ice_Emote_N}
  n21 = Transform(None=@n20.RGB)
  n22 = RotateAboutAxis(NormalizedRotationAxis=@n6, RotationAngle=@n11.RotationOverLifetime, PivotPoint=Constant(), Position=@n21)
  Transform(None=Normalize(VectorInput=Add(A=@n22, B=@n21))) {transformSourceType=TRANSFORMSOURCE_World, transformType=TRANSFORM_Tangent}
  VectorParameter() {defaultValue=<rgba 0.0378 0.0516 0.0833 1>}
 ]
}
