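# Translucent, unlit, two-sided glow material for the Sculptor player emote FX, enabled for Niagara sprites.
# Particle color RGB passes through MF_FX_EmissiveAdjustments for emissive.
# Opacity multiplies particle alpha, the glow texture's red channel and a switchable angle fade that dims faces seen edge-on.
# A shading path switch applies depth fade to default and deferred paths, omitting it on forward and mobile.
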Material /Game/Animation/Game/MainPlayer/Emotes/Sculptor/FX/M_Emote_Sculptor_Glow
graph {
 properties {
  bUsedWithNiagaraSprites=true
  blendMode=BLEND_Translucent
  pixelDepthOffsetMode=PDOM_Legacy
  shadingModel=MSM_Unlit
  twoSided=true
  usedShadingModels=Unlit
 }
 nodes [
  n1 = ParticleColor()
  MaterialFunctionCall(Emissive=@n1.RGB) {materialFunction=&MF_FX_EmissiveAdjustments}
  n2 = Abs(None=DotProduct(A=VertexNormalWS(), B=CameraVectorWS()))
  n3 = Multiply(
   A=Multiply(
    A=@n1.A
    B=TextureSampleParameter2D() {parameterName=GlowTexture, texture=&/T_Emote_Sculptor_2024_GlowFront}.R
   )
   B=StaticSwitchParameter(True=Power(Base=@n2) {constExponent=3}, False=Constant() {r=1}) {defaultValue=true, parameterName=UseAngleFade}
  )
  n4 = DepthFade(Opacity=@n3, FadeDistance=ScalarParameter() {defaultValue=3, parameterName=FadeDistance}) {fadeDistanceDefault=3}
  ShadingPathSwitch(Default=@n4, Deferred=@n4, Forward=@n3, Mobile=@n3)
 ]
}
