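# Niagara system for the main player's Sculptor emote, drawing a heavy smoke puff as faint blue sprites.
# Its single AirGlow emitter pairs a small burst with a slow spawn rate, launching particles from a small box volume.
# Sprites play curve-driven SubUV animation, grow, fade in and out, and sink under drag and gravity.
# A time-of-day function boosts the blue tint at night; User.NormalizedNotifyProgress is exposed but unreferenced in the modules.
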NiagaraSystem /Game/Animation/Game/MainPlayer/Emotes/Sculptor/FX/NS_Emote_Sculptor_HeavySmoke
dependencies {
 usedDataInterfaces [&NiagaraDataInterfaceSpriteRendererInfo, &NiagaraDataInterfaceCurve]
 usedDynamicInputs []
 usedModules [&SystemState]
}
emitters {AirGlow {
  emitterUpdateScript [
   &EmitterState()
   &SpawnRate(SpawnRate=0.8, Use Spawn Probability=false)
   &SpawnBurst_Instantaneous(Spawn Count=2, Use Spawn Probability=false)
  ]
  particleSpawnScript [
   &InitializeParticle(
    Color=<rgba 0.4263 0.7193 1 0.05>
    Lifetime=&RandomRangeFloat(Minimum=1.4, Maximum=2.4)
    Position Offset=<0 0 10>
    Sprite Rotation Mode=::NewEnumerator2
    Sprite Size Mode=::NewEnumerator3
    Uniform Sprite Size=&RandomRangeFloat(Minimum=60, Maximum=80)
    UsePositionOffset=true
    Write Material Random=true
   ) {module=InitializeParticle001}
   &DynamicMaterialParameters(
    Index 0 Param 1=2
    Module.Param0WriteEnabled=true
    Module.Param1WriteEnabled=true
    Module.Param2WriteEnabled=true
    Module.Param3WriteEnabled=true
   )
   &CameraOffset(Camera Offset Amount=10)
   &ShapeLocation(
    Box Midpoint=<0.5 0.5 0.3>
    Box Size=<15 15 30>
    MaskSpawnGroup=false
    Override Seed=false
    Shape Primitive=::NewEnumerator2
    Surface Only=false
    Surface Only Band Thickness=1
    UseOwnerScale=true
   )
   &AddVelocity(
    Module.Use Falloff Distance=false
    Override Seed=false
    Velocity Mode=::NewEnumerator1
    Velocity Speed=&RandomRangeFloat(Minimum=25, Maximum=60)
   )
   &Color(Color=&MultiplyLinearColor(
     Linear Color=$Particles.Color
     Scale Factor=&MakeVector4_FromLinearColor(LinearColor=&FN_Select_Color_from_Time_Of_Day(Morning=<rgba 1 1 1 1>, Day=<rgba 1 1 1 1>, Evening=<rgba 1 1 1 1>, Night=<rgba 0.7111 1.7727 3 1>))
    ))
  ]
  particleUpdateScript [
   &ParticleState()
   &SubUVAnimation(
    "Sprite Renderer "=DataInterface() {propertyValues {SpriteRenderer=&self:AirGlow_3.NiagaraSpriteRendererProperties_2}}
    SubUV Animation Mode:ENiagara_SubUVLookupModeV2=::NewEnumerator1
    SubUV Curve=DataInterface() {propertyValues {Curve {keys [
        {arriveTangent=1.6702, interpMode=RCIM_Cubic, leaveTangent=1.6702, tangentMode=RCTM_User, time=0, value=0}
        {arriveTangent=1.562, interpMode=RCIM_Cubic, leaveTangent=1.562, tangentMode=RCTM_User, time=1, value=0.85}
       ]}}}
    UseEndFrame=false
    UseStartFrame=false
   )
   &ScaleSpriteSize(Scale Factor=&Vector2DFromFloat(Value=&FloatFromCurve(FloatCurve=DataInterface() {propertyValues {Curve {keys [
          {arriveTangent=0.4241, interpMode=RCIM_Cubic, leaveTangent=0.4241, tangentMode=RCTM_User, time=0, value=0.4}
          {arriveTangent=-0.405, tangentMode=RCTM_User, time=1, value=1}
         ]}}})))
   &ScaleColor(
    Scale Alpha=&FloatFromCurve(FloatCurve=DataInterface() {propertyValues {Curve {keys [<at 0 0>, {interpMode=RCIM_Cubic, leaveTangent=-2.734, tangentMode=RCTM_Break, time=0.4, value=0.75}, <at 1 0>]}}})
    ScaleA=true
    ScaleRGB=true
   ) {module=ScaleColor001}
   &Drag(Drag=2, Use Linear Drag=true, Use Rotational Drag=true)
   &GravityForce(Gravity=<0 0 -50>)
   &SolveForcesAndVelocity(Clamp Velocity=false, Limit Acceleration=false)
  ]
  rendererClasses [&NiagaraSpriteRendererProperties]
  renderers [{rendererClass=&NiagaraSpriteRendererProperties}]
 }}
summary {
 systemName=NS_Emote_Sculptor_HeavySmoke
 emitters [{emitterName=AirGlow, rendererClasses [&NiagaraSpriteRendererProperties]}]
 userVariables [User.NormalizedNotifyProgress {type=NiagaraFloat, defaultValue=0, description=""}]
}
